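# Translucent cloak material for a device in the Build Radar Tower secondary mission, fading between a refractive, rim-only look and an opaque white-to-blue glow.
# Custom primitive data supplies start time and direction to a timed "Cloak Transition" value that blends opacity, emissive and refraction.
# Camera-facing virtual-plane noise distorts a second scrolling noise sample, which hue-shifts the base color.
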Material /Game/Missions/Secondary/BuildRadarTower/Materials/M_DeviceCloaked
graph {
 properties {
  bUseMaterialAttributes=true
  bUsedWithInstancedStaticMeshes=true
  bUsedWithNanite=true
  blendMode=BLEND_Translucent
  pixelDepthOffsetMode=PDOM_Legacy
  refractionMethod=RM_IndexOfRefraction
  usedShadingModels=Default Lit
 }
 nodes [
  n1 = Time()
  n2 = ScalarParameter() {bUseCustomPrimitiveData=true, parameterName=CloakingTime}
  n3 = Saturate(None=Divide(A=Subtract(A=@n1, B=@n2), B=ScalarParameter() {defaultValue=0.5, parameterName=Transition Duration}))
  NamedRerouteDeclaration(None=LinearInterpolate(A=OneMinus(None=@n3), B=@n3, Alpha=ScalarParameter() {bUseCustomPrimitiveData=true, parameterName=Transitioning, primitiveDataIndex=1})) {name=Cloak Transition, nodeColor=<rgba 1 0 0.3984 1>}
  n4 = ScalarParameter() {defaultValue=0.3, parameterName=perma}
  n5 = MaterialFunctionCall(
   Plane Normal Axis=MaterialFunctionCall() {materialFunction=&CameraDirectionVector}.Result
   UV Scale=Constant() {r=500}
  ) {materialFunction=&VirtualPlaneCoordinates}
  n6 = Add(
   A=TextureSample(UVs=Panner(Coordinate=@n5.UVs) {speedY=0.001}) {texture=/Game/Effects/FogVolume/noise_flamy_Tile}.R
   B=Divide(A=Time()) {constB=3}
  )
  n7 = Constant() {r=1}
  n8 = TextureSample(UVs=Add(A=TextureCoordinate() {uTiling=5, vTiling=5}, B=Fmod(A=@n6, B=@n7))) {texture=&noise_flamy_Tile}
  n9 = TextureSampleParameter2D() {
   group=Material_1
   parameterName=Normals
   samplerSource=SSM_Wrap_WorldGroupSettings
   samplerType=SAMPLERTYPE_Normal
   sortPriority=-5
   texture=/Engine/EngineMaterials/FlatNormal
  }
  n10 = OneMinus(None=Fresnel() {baseReflectFraction=0})
  n11 = NamedRerouteUsage()
  n12 = LinearInterpolate(
   A=Constant3Vector() {constant=<rgba 1 1 1 0>}
   B=Constant3Vector() {constant=<rgba 0 0.6058 1 0>}
   Alpha=Fresnel(Normal=PixelNormalWS()) {exponent=1}
  )
  SetMaterialAttributes!(
   MaterialAttributes
   Base Color=MaterialFunctionCall(Hue Shift Percentage=Add(A=@n8.R, B=@n4), Texture=Constant3Vector() {constant=<rgba 1 0.5 0.5 0>}) {materialFunction=&HueShift}.Result
   Opacity=LinearInterpolate(A=Fresnel() {baseReflectFraction=0.01, exponent=15}, Alpha=NamedRerouteUsage())
   Specular
   Roughness
   Normal=@n9.RGB
   Emissive Color=LinearInterpolate(B=Multiply(A=@n12) {constB=4}, Alpha=NamedRerouteUsage())
   Refraction=LinearInterpolate(Alpha=Multiply(A=@n10, B=@n11)) {constA=1.015}
  ) {attributeSetTypes [
    #1
    #2
    #3
    #4
    #5
    #6
    #7
   ]}
  Transform(None=@n9.RGB)
 ]
}
