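# Lit material for an uncloaked device in the Build Radar Tower mission, adding a glowing cloak-transition effect over standard diffuse, mask, normal and emissive maps.
# Custom primitive data supplies start time and direction for a timed transition value.
# That value thresholds scrolling world-aligned noise, dotted with a hue-cycling vector, into a dissolve mask and edge band.
# The result, Fresnel-tinted white to cyan, feeds only emissive, leaving opacity untouched.
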Material /Game/Missions/Secondary/BuildRadarTower/Materials/M_DeviceUncloaked
graph {
 properties {bUseMaterialAttributes=true, pixelDepthOffsetMode=PDOM_Legacy, usedShadingModels=Default Lit}
 nodes [
  n1 = Time()
  n2 = ScalarParameter() {bUseCustomPrimitiveData=true, parameterName=CloakingTime}
  n3 = Saturate(None=Divide(A=Subtract(A=@n1, B=@n2), B=ScalarParameter() {defaultValue=0.5, parameterName=Transition Duration}))
  n4 = OneMinus(None=@n3)
  n5 = ScalarParameter() {bUseCustomPrimitiveData=true, parameterName=Transitioning, primitiveDataIndex=1}
  NamedRerouteDeclaration(None=Power(Base=LinearInterpolate(A=@n4, B=@n3, Alpha=@n5))) {name=Cloaking Transition, nodeColor=<rgba 0.0078 1 0 1>}
  NamedRerouteUsage()
  ScalarParameter() {defaultValue=0.192, sliderMax=1}
  n6 = TextureSampleParameter2D() {
   group=Material_1
   parameterName=Diffuse
   samplerSource=SSM_Wrap_WorldGroupSettings
   sortPriority=-5
   texture=/Game/Global/Textures/Default/Blanks/T_MidgreysRGB
  }
  n7 = TextureSampleParameter2D() {
   group=Material_1
   parameterName=SpecularMasks
   samplerSource=SSM_Wrap_WorldGroupSettings
   samplerType=SAMPLERTYPE_Masks
   sortPriority=-5
   texture=&T_Fortnite_Default_S
  }
  n8 = TextureSampleParameter2D() {
   group=Material_1
   parameterName=Normals
   samplerSource=SSM_Wrap_WorldGroupSettings
   samplerType=SAMPLERTYPE_Normal
   sortPriority=-5
   texture=/Engine/EngineMaterials/FlatNormal
  }
  n9 = VectorParameter() {parameterName=Emissive Color, defaultValue=<rgba 1 1 1 1>}
  n10 = TextureSampleParameter2D() {
   group=Material_1
   parameterName=Emissive
   samplerSource=SSM_Wrap_WorldGroupSettings
   samplerType=SAMPLERTYPE_LinearGrayscale
   sortPriority=-5
   texture=/Game/Global/Textures/Default/Blanks/T_BlackLinearGrayscale
  }
  n11 = Constant3Vector() {constant=<rgba 1 1 1 0>}
  n12 = Constant3Vector() {constant=<rgba 0 0.6058 1 0>}
  n13 = Fresnel() {exponent=1}
  n14 = NamedRerouteUsage()
  n15 = OneMinus(None=@n14)
  n16 = TextureObject() {texture=/Game/Athena/UI/Challenges/Art/T_UI_DataChip_Noise}
  n17 = Constant3Vector() {constant=<rgba 75 75 75 0>}
  n18 = Time()
  n19 = Add(
   A=WorldPosition().XYZ
   B=MaterialFunctionCall(Z=Multiply(A=@n18) {constB=-20}) {materialFunction=&MakeFloat3}.Result
  )
  n20 = Divide(A=Time()) {constB=1}
  n21 = Constant3Vector() {constant=<rgba 1 0.7 0.7 0>}
  n22 = DotProduct(
   A=Normalize(VectorInput=MaterialFunctionCall(TextureObject=@n16, TextureSize=@n17, WorldPosition=@n19) {materialFunction=&WorldAlignedTexture}.XYZ Texture)
   B=Normalize(VectorInput=MaterialFunctionCall(Hue Shift Percentage=@n20, Texture=@n21) {materialFunction=&HueShift}.Result)
  )
  n23 = SmoothStep(Min=@n15, Max=@n15, Value=@n22)
  n24 = Multiply(A=@n23, B=LinearInterpolate(Alpha=Power(Base=@n14)) {constA=0.2, constB=4}) {constB=0.2}
  n25 = Constant() {r=-0.05}
  n26 = Constant() {r=1.05}
  n27 = Constant()
  n28 = If(
   A=@n22
   B=OneMinus(None=MaterialFunctionCall(Input=@n14, Target Low=@n25, Target High=@n26) {materialFunction=&RemapValueRange}.Result)
   "A > B"=@n27
   "A == B"=Constant() {r=1}
   "A < B"=@n27
  ) {equalsThreshold=0.05}
  n29 = Multiply(A=Multiply(A=LinearInterpolate(A=@n11, B=@n12, Alpha=@n13), B=Add(A=@n24, B=@n28)), B=@n23)
  SetMaterialAttributes!(
   MaterialAttributes
   Base Color=@n6.RGB
   Metallic=@n7.G
   Specular=@n7.G
   Roughness=@n7.G
   Normal=@n8.RGB
   Emissive Color=Add(A=Multiply(A=@n10.RGB, B=@n9.RGB), B=Add(B=@n29))
  ) {attributeSetTypes [
    #1
    #2
    #3
    #4
    #5
    #6
   ]}
 ]
}
